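# Material function for the SpruceBark player body that adjusts base color, metallic, specular and roughness in fur-masked areas.
# Base color gets a saturation control through inverted Desaturation, while the other three take additive biases clamped to 0–1.
# A parameter-selected channel of the TechArtMask input blends the adjusted attributes over the incoming ones.
# The UseFurParamAdjustments static switch outputs either this blend or the unmodified material.
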MaterialFunction /BRCosmetics/Characters/Player/Male/Medium/Bodies/M_MED_SpruceBark/Materials/MF_Fur_ParamAdjustments_SpruceBark
graph {nodes [
  n1 = NamedRerouteUsage()
  n2 = SetMaterialAttributes!(
   MaterialAttributes=NamedRerouteUsage()
   Roughness=NamedRerouteUsage()
   Specular=NamedRerouteUsage()
   Metallic=NamedRerouteUsage()
   BaseColor=NamedRerouteUsage()
  ) {attributeSetTypes [
    #1
    #2
    #3
    #4
   ]}
  n3 = NamedRerouteUsage()
  FunctionOutput(None=StaticSwitchParameter(True=BlendMaterialAttributes(A=@n1, B=@n2, Alpha=@n3), False=@n1) {group=Fur_Shared, parameterName=UseFurParamAdjustments, sortPriority=-5})
  n4 = GetMaterialAttributes(None=NamedRerouteDeclaration(None=FunctionInput() {bUsePreviewValueAsDefault=true, inputName=MA, inputType=FunctionInput_MaterialAttributes}) {name=IN_Material, nodeColor=<rgba 0.125 0 1 1>}) {attributeGetTypes [
    #2
    #1
    #3
    #4
   ]}
  NamedRerouteDeclaration(None=@n4.BaseColor) {name=In_BC, nodeColor=<rgba 1 0 0.5396 1>}
  n5 = ScalarParameter() {defaultValue=1.1, group=Fur_Shared, parameterName=FurSaturation}
  NamedRerouteDeclaration(None=Desaturation(None=NamedRerouteUsage(), Fraction=OneMinus(None=@n5))) {name=Fur_BC, nodeColor=<rgba 0.5938 0 1 1>}
  NamedRerouteDeclaration(None=@n4.MaterialAttributes) {name=In_MA, nodeColor=<rgba 1 0 0.5396 1>}
  NamedRerouteDeclaration(None=@n4.Specular) {name=In_Spec, nodeColor=<rgba 1 0 0.5396 1>}
  NamedRerouteDeclaration(None=@n4.Roughness) {name=In_Roughness, nodeColor=<rgba 1 0 0.5396 1>}
  NamedRerouteDeclaration(None=@n4.Metallic) {name=In_Metallic, nodeColor=<rgba 1 0 0.5396 1>}
  n6 = NamedRerouteUsage()
  n7 = ScalarParameter() {group=Fur_Shared, parameterName=MetallicBias, sortPriority=-3}
  NamedRerouteDeclaration(None=Saturate(None=Add(A=@n6, B=@n7))) {name=Fur_Metallic, nodeColor=<rgba 0.5938 0 1 1>}
  n8 = NamedRerouteUsage()
  n9 = ScalarParameter() {group=Fur_Shared, parameterName=SpecularBias, sortPriority=-3}
  NamedRerouteDeclaration(None=Saturate(None=Add(A=@n8, B=@n9))) {name=Fur_Specular, nodeColor=<rgba 0.5938 0 1 1>}
  n10 = NamedRerouteUsage()
  n11 = ScalarParameter() {group=Fur_Shared, parameterName=RoughnessBias, sortPriority=-3}
  NamedRerouteDeclaration(None=Saturate(None=Add(A=@n10, B=@n11))) {name=Fur_Roughness, nodeColor=<rgba 0.5938 0 1 1>}
  n12 = FunctionInput() {bUsePreviewValueAsDefault=true, inputName=TechArtMask, sortPriority=1, previewValue=<1 1 1 1>}
  NamedRerouteDeclaration(None=ChannelMaskParameter(None=@n12) {group=Fur_Shared, parameterName=ParamAdjustments_FurMaskChannel, sortPriority=-4}.R) {name=In_FurMask, nodeColor=<rgba 1 0 0.5396 1>}
 ]}
