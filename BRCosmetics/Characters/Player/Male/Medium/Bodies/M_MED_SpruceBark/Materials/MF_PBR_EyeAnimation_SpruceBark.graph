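# Material function for the SpruceBark character body that draws animated, material-driven pupils onto an incoming material.
# It splits left and right eyes by UV range, then offsets, rotates and scales pupil UVs from look-pose parameters, with optional distance/FOV scaling.
# Pupil diffuse and optional SRM and normal textures are blended in through an eye mask.
# A TemporalResponsiveness output drives TSR invalidation when look direction changes.
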MaterialFunction /BRCosmetics/Characters/Player/Male/Medium/Bodies/M_MED_SpruceBark/Materials/MF_PBR_EyeAnimation_SpruceBark
graph {expressions {
  MaterialExpressionAdd_0 {
   class=MaterialExpressionAdd
   inputs [
    A {from=MaterialExpressionAdd_11}
    B {from=MaterialExpressionLinearInterpolate_1}
   ]
  }
  MaterialExpressionAdd_1 {
   class=MaterialExpressionAdd
   inputs [
    A {from=MaterialExpressionAdd_9}
    B {from=MaterialExpressionLinearInterpolate_2}
   ]
  }
  MaterialExpressionAdd_11 {
   class=MaterialExpressionAdd
   inputs [
    A {from=MaterialExpressionScalarParameter_20}
    B {from=MaterialExpressionMultiply_13}
   ]
  }
  MaterialExpressionAdd_15 {
   class=MaterialExpressionAdd
   inputs [
    A {from=MaterialExpressionMaterialFunctionCall_2, output=R}
    B {from=""}
   ]
   properties {constB=-1}
  }
  MaterialExpressionAdd_16 {
   class=MaterialExpressionAdd
   inputs [
    A {from=MaterialExpressionMaterialFunctionCall_16, output=G}
    B {from=""}
   ]
   properties {constB=-0.5}
  }
  MaterialExpressionAdd_17 {
   class=MaterialExpressionAdd
   inputs [
    A {from=MaterialExpressionMultiply_5}
    B {from=""}
   ]
   properties {constB=0.5}
  }
  MaterialExpressionAdd_19 {
   class=MaterialExpressionAdd
   inputs [
    A {from=MaterialExpressionComponentMask_2}
    B {from=MaterialExpressionConvert_2}
   ]
  }
  MaterialExpressionAdd_2 {
   class=MaterialExpressionAdd
   inputs [
    A {from=MaterialExpressionStaticSwitchParameter_11}
    B {from=MaterialExpressionScalarParameter_5}
   ]
  }
  MaterialExpressionAdd_20 {
   class=MaterialExpressionAdd
   inputs [
    A {from=MaterialExpressionNamedRerouteUsage_24}
    B {from=""}
   ]
   properties {constB=0}
  }
  MaterialExpressionAdd_22 {
   class=MaterialExpressionAdd
   inputs [
    A {from=MaterialExpressionMultiply_35}
    B {from=""}
   ]
   properties {constB=0.5}
  }
  MaterialExpressionAdd_23 {
   class=MaterialExpressionAdd
   inputs [
    A {from=MaterialExpressionComponentMask_3}
    B {from=MaterialExpressionConvert_3}
   ]
  }
  MaterialExpressionAdd_26 {
   class=MaterialExpressionAdd
   inputs [
    A {from=MaterialExpressionLinearInterpolate_28}
    B {from=MaterialExpressionLinearInterpolate_18}
   ]
  }
  MaterialExpressionAdd_4 {
   class=MaterialExpressionAdd
   inputs [
    A {from=MaterialExpressionMaterialFunctionCall_17, output=Result}
    B {from=MaterialExpressionMultiply_27}
   ]
   properties {desc=move right eye to UV space 0-1}
  }
  MaterialExpressionAdd_5 {
   class=MaterialExpressionAdd
   inputs [
    A {from=MaterialExpressionAdd_4}
    B {from=MaterialExpressionAppendVector_0}
   ]
  }
  MaterialExpressionAdd_9 {
   class=MaterialExpressionAdd
   inputs [
    A {from=MaterialExpressionScalarParameter_26}
    B {from=MaterialExpressionMultiply_16}
   ]
  }
  MaterialExpressionAppendVector_0 {
   class=MaterialExpressionAppendVector
   inputs [
    A {from=MaterialExpressionMultiply_40, output=""}
    B {from=MaterialExpressionMultiply_41, output=""}
   ]
  }
  MaterialExpressionBlendMaterialAttributes_0 {
   class=MaterialExpressionBlendMaterialAttributes
   inputs [
    A {from=MaterialExpressionNamedRerouteUsage_52, output=""}
    B {from=MaterialExpressionSetMaterialAttributes_0, output=""}
    Alpha {from=MaterialExpressionNamedRerouteUsage_55, output=""}
   ]
   properties {}
  }
  MaterialExpressionCeil_1 {class=MaterialExpressionCeil, inputs [None {from=MaterialExpressionMaterialFunctionCall_4, output=G}], properties {}}
  MaterialExpressionClamp_2 {
   class=MaterialExpressionClamp
   inputs [
    None {from=MaterialExpressionAdd_9, output=""}
    Min {from="", output=""}
    Max {from="", output=""}
   ]
   properties {minDefault=-1}
  }
  MaterialExpressionClamp_3 {
   class=MaterialExpressionClamp
   inputs [
    None {from=MaterialExpressionAdd_11, output=""}
    Min {from="", output=""}
    Max {from="", output=""}
   ]
   properties {minDefault=-1}
  }
  MaterialExpressionComponentMask_0 {class=MaterialExpressionComponentMask, inputs [None {from=MaterialExpressionStaticSwitchParameter_15}], properties {b=true}}
  MaterialExpressionComponentMask_1 {class=MaterialExpressionComponentMask, inputs [None {from=MaterialExpressionStaticSwitchParameter_15}], properties {r=true}}
  MaterialExpressionComponentMask_2 {class=MaterialExpressionComponentMask, inputs [None {from=MaterialExpressionVectorParameter_5, output=RGB}], properties {g=true, r=true}}
  MaterialExpressionComponentMask_3 {class=MaterialExpressionComponentMask, inputs [None {from=MaterialExpressionVectorParameter_6, output=RGB}], properties {g=true, r=true}}
  MaterialExpressionComponentMask_4 {class=MaterialExpressionComponentMask, inputs [None {from=MaterialExpressionStaticSwitchParameter_15}], properties {g=true}}
  MaterialExpressionConstant2Vector_0 {class=MaterialExpressionConstant2Vector, properties {r=-1}}
  MaterialExpressionConstantBiasScale_0 {class=MaterialExpressionConstantBiasScale, inputs [None {from=MaterialExpressionClamp_3}]}
  MaterialExpressionConstant_3 {class=MaterialExpressionConstant, properties {r=0.4}}
  MaterialExpressionConvert_2 {
   class=MaterialExpressionConvert
   inputs [
    None {from=MaterialExpressionScalarParameter_46}
    None {from=MaterialExpressionScalarParameter_47}
   ]
   properties {
    nodeName=Make Vector2
    convertInputs [
     {
      type=Scalar
      expressionInput {
       expression=&self:MaterialExpressionScalarParameter_46
       inputName=None
       mask=0
       maskA=0
       maskB=0
       maskG=0
       maskR=0
       outputIndex=0
      }
     }
     {
      type=Scalar
      expressionInput {
       expression=&self:MaterialExpressionScalarParameter_47
       inputName=None
       mask=0
       maskA=0
       maskB=0
       maskG=0
       maskR=0
       outputIndex=0
      }
     }
    ]
    convertOutputs [{type=Vector2}]
   }
  }
  MaterialExpressionConvert_3 {
   class=MaterialExpressionConvert
   inputs [
    None {from=MaterialExpressionScalarParameter_48}
    None {from=MaterialExpressionScalarParameter_49}
   ]
   properties {
    nodeName=Make Vector2
    convertInputs [
     {
      type=Scalar
      expressionInput {
       expression=&self:MaterialExpressionScalarParameter_48
       inputName=None
       mask=0
       maskA=0
       maskB=0
       maskG=0
       maskR=0
       outputIndex=0
      }
     }
     {
      type=Scalar
      expressionInput {
       expression=&self:MaterialExpressionScalarParameter_49
       inputName=None
       mask=0
       maskA=0
       maskB=0
       maskG=0
       maskR=0
       outputIndex=0
      }
     }
    ]
    convertOutputs [{type=Vector2}]
   }
  }
  MaterialExpressionFloor_0 {class=MaterialExpressionFloor, inputs [None {from=MaterialExpressionAdd_20, output=""}], properties {}}
  MaterialExpressionFloor_1 {class=MaterialExpressionFloor, inputs [None {from=MaterialExpressionMaterialFunctionCall_4, output=R}], properties {desc=LeftEye}}
  MaterialExpressionFunctionInput_0 {
   class=MaterialExpressionFunctionInput
   inputs [Preview {from="", output=""}]
   properties {bUsePreviewValueAsDefault=true, inputName=Mask, inputType=FunctionInput_Scalar, sortPriority=20, previewValue=<1 0 0 1>}
  }
  MaterialExpressionFunctionInput_1 {
   class=MaterialExpressionFunctionInput
   inputs [Preview {from=MaterialExpressionStaticBool_0, output=""}]
   properties {
    bUsePreviewValueAsDefault=true
    description="if true (default), uses the second UV channel for eye mapping. If false, uses the first UV channel."
    inputName=Use UV2
    inputType=FunctionInput_StaticBool
    sortPriority=15
    previewValue=<1 0 0 1>
   }
  }
  MaterialExpressionFunctionInput_4 {
   class=MaterialExpressionFunctionInput
   inputs [Preview {from="", output=""}]
   properties {inputName=Material, inputType=FunctionInput_MaterialAttributes}
  }
  MaterialExpressionFunctionOutput_3 {class=MaterialExpressionFunctionOutput, inputs [None {from=MaterialExpressionStaticSwitchParameter_14, output=""}]}
  MaterialExpressionFunctionOutput_5 {
   class=MaterialExpressionFunctionOutput
   inputs [None {from=MaterialExpressionStaticSwitchParameter_12, output=""}]
   properties {
    description=Connect directly to the Temporal Responsiveness special material output node if this toon character only has material-driven pupils.
    outputName=TemporalResponsiveness
    sortPriority=20
   }
  }
  MaterialExpressionGetMaterialAttributes_0 {
   class=MaterialExpressionGetMaterialAttributes
   inputs [None {from=MaterialExpressionFunctionInput_4, output=""}]
   properties {attributeGetTypes [
     #1
     #2
     #3
     #4
     #5
    ]}
  }
  MaterialExpressionInverseLinearInterpolate_3 {
   class=MaterialExpressionInverseLinearInterpolate
   inputs [
    A {from="", output=""}
    B {from="", output=""}
    Value {from=MaterialExpressionMaterialFunctionCall_19, output=FOV angle in degrees}
   ]
   properties {constA=45, constB=90}
  }
  MaterialExpressionLinearInterpolate_0 {
   class=MaterialExpressionLinearInterpolate
   inputs [
    A {from=MaterialExpressionStaticSwitchParameter_10}
    B {from=MaterialExpressionScalarParameter_30}
    Alpha {from=MaterialExpressionMultiply_21}
   ]
  }
  MaterialExpressionLinearInterpolate_1 {
   class=MaterialExpressionLinearInterpolate
   inputs [
    A {from=MaterialExpressionScalarParameter_1}
    B {from=MaterialExpressionScalarParameter_2}
    Alpha {from=MaterialExpressionNamedRerouteUsage_1}
   ]
  }
  MaterialExpressionLinearInterpolate_13 {
   class=MaterialExpressionLinearInterpolate
   inputs [
    A {from=MaterialExpressionReroute_0}
    B {from=MaterialExpressionMaterialFunctionCall_12, output=Result}
    Alpha {from=MaterialExpressionLinearInterpolate_21}
   ]
   properties {desc=rotate right eye UVs}
  }
  MaterialExpressionLinearInterpolate_18 {
   class=MaterialExpressionLinearInterpolate
   inputs [
    A {from=MaterialExpressionAdd_19}
    B {from=MaterialExpressionAdd_23}
    Alpha {from=MaterialExpressionNamedRerouteUsage_32}
   ]
  }
  MaterialExpressionLinearInterpolate_19 {
   class=MaterialExpressionLinearInterpolate
   inputs [
    A {from=""}
    B {from=""}
    Alpha {from=MaterialExpressionConstantBiasScale_0}
   ]
  }
  MaterialExpressionLinearInterpolate_2 {
   class=MaterialExpressionLinearInterpolate
   inputs [
    A {from=MaterialExpressionScalarParameter_3}
    B {from=MaterialExpressionScalarParameter_4}
    Alpha {from=MaterialExpressionNamedRerouteUsage_4}
   ]
  }
  MaterialExpressionLinearInterpolate_21 {
   class=MaterialExpressionLinearInterpolate
   inputs [
    A {from=""}
    B {from=MaterialExpressionNamedRerouteUsage_23}
    Alpha {from=MaterialExpressionScalarParameter_34}
   ]
   properties {constB=0}
  }
  MaterialExpressionLinearInterpolate_22 {
   class=MaterialExpressionLinearInterpolate
   inputs [
    A {from=MaterialExpressionScalarParameter_40}
    B {from=MaterialExpressionScalarParameter_41}
    Alpha {from=MaterialExpressionNamedRerouteUsage_31}
   ]
  }
  MaterialExpressionLinearInterpolate_23 {
   class=MaterialExpressionLinearInterpolate
   inputs [
    A {from=MaterialExpressionScalarParameter_44}
    B {from=MaterialExpressionScalarParameter_43}
    Alpha {from=MaterialExpressionNamedRerouteUsage_40}
   ]
  }
  MaterialExpressionLinearInterpolate_24 {
   class=MaterialExpressionLinearInterpolate
   inputs [
    A {from=""}
    B {from=""}
    Alpha {from=MaterialExpressionMultiply_22}
   ]
   properties {constA=0.4}
  }
  MaterialExpressionLinearInterpolate_25 {
   class=MaterialExpressionLinearInterpolate
   inputs [
    A {from=MaterialExpressionScalarParameter_51}
    B {from=MaterialExpressionScalarParameter_36}
    Alpha {from=MaterialExpressionPower_5}
   ]
   properties {constA=0.5, constB=0.5}
  }
  MaterialExpressionLinearInterpolate_28 {
   class=MaterialExpressionLinearInterpolate
   inputs [
    A {from=MaterialExpressionRotator_2}
    B {from=MaterialExpressionRotator_3}
    Alpha {from=MaterialExpressionNamedRerouteUsage_0}
   ]
  }
  MaterialExpressionLinearInterpolate_29 {
   class=MaterialExpressionLinearInterpolate
   inputs [
    A {from=MaterialExpressionGetMaterialAttributes_0, output=BaseColor}
    B {from=MaterialExpressionTextureSampleParameter2D_1, output=RGB}
    Alpha {from=MaterialExpressionTextureSampleParameter2D_1, output=RGB}
   ]
  }
  MaterialExpressionMaterialFunctionCall_0 {
   class=MaterialExpressionMaterialFunctionCall
   inputs [
    UVs {from=MaterialExpressionAdd_5}
    Texture Scale {from=MaterialExpressionStaticSwitchParameter_0}
   ]
   properties {materialFunction=&ScaleUVsByCenter}
  }
  MaterialExpressionMaterialFunctionCall_1 {
   class=MaterialExpressionMaterialFunctionCall
   inputs [
    Fade Length {from=MaterialExpressionScalarParameter_32}
    Fade Offset {from=MaterialExpressionScalarParameter_31}
    For Vertex Shader {from=""}
   ]
   properties {materialFunction=&CameraDepthFade}
  }
  MaterialExpressionMaterialFunctionCall_10 {
   class=MaterialExpressionMaterialFunctionCall
   inputs [Float2 {from=MaterialExpressionMaterialFunctionCall_0, output=UVs}]
   properties {materialFunction=&BreakOutFloat2Components}
  }
  MaterialExpressionMaterialFunctionCall_11 {
   class=MaterialExpressionMaterialFunctionCall
   inputs [
    X {from=MaterialExpressionAdd_22}
    Y {from=MaterialExpressionMaterialFunctionCall_10, output=G}
   ]
   properties {materialFunction=&MakeFloat2}
  }
  MaterialExpressionMaterialFunctionCall_12 {
   class=MaterialExpressionMaterialFunctionCall
   inputs [
    X {from=MaterialExpressionMaterialFunctionCall_9, output=G}
    Y {from=MaterialExpressionMaterialFunctionCall_9, output=G}
   ]
   properties {materialFunction=&MakeFloat2}
  }
  MaterialExpressionMaterialFunctionCall_13 {
   class=MaterialExpressionMaterialFunctionCall
   inputs [
    R {from=MaterialExpressionGetMaterialAttributes_0, output=Specular}
    G {from=MaterialExpressionGetMaterialAttributes_0, output=Specular}
    B {from=MaterialExpressionGetMaterialAttributes_0, output=Specular}
    A {from=""}
   ]
   properties {materialFunction=&AppendMany}
  }
  MaterialExpressionMaterialFunctionCall_16 {
   class=MaterialExpressionMaterialFunctionCall
   inputs [Float2 {from=MaterialExpressionAdd_26}]
   properties {materialFunction=&BreakOutFloat2Components}
  }
  MaterialExpressionMaterialFunctionCall_17 {
   class=MaterialExpressionMaterialFunctionCall
   inputs [
    X {from=MaterialExpressionMaterialFunctionCall_16, output=R}
    Y {from=MaterialExpressionAdd_17}
   ]
   properties {materialFunction=&MakeFloat2}
  }
  MaterialExpressionMaterialFunctionCall_19 {
   class=MaterialExpressionMaterialFunctionCall
   inputs []
   properties {materialFunction=/Engine/Functions/Engine_MaterialFunctions02/Utility/FOV}
  }
  MaterialExpressionMaterialFunctionCall_2 {
   class=MaterialExpressionMaterialFunctionCall
   inputs [Float2 {from=MaterialExpressionFloor_0}]
   properties {materialFunction=&BreakOutFloat2Components}
  }
  MaterialExpressionMaterialFunctionCall_4 {
   class=MaterialExpressionMaterialFunctionCall
   inputs [Float2 {from=MaterialExpressionNamedRerouteUsage_22}]
   properties {materialFunction=&BreakOutFloat2Components}
  }
  MaterialExpressionMaterialFunctionCall_7 {
   class=MaterialExpressionMaterialFunctionCall
   inputs [Float2 {from=MaterialExpressionMultiply_3}]
   properties {materialFunction=&BreakOutFloat2Components}
  }
  MaterialExpressionMaterialFunctionCall_8 {
   class=MaterialExpressionMaterialFunctionCall
   inputs [
    BaseNormal {from=MaterialExpressionNamedRerouteUsage_53}
    AdditionalNormal {from=MaterialExpressionTextureSampleParameter2D_3, output=RGB}
   ]
   properties {materialFunction=&BlendAngleCorrectedNormals}
  }
  MaterialExpressionMaterialFunctionCall_9 {
   class=MaterialExpressionMaterialFunctionCall
   inputs [Float2 {from=MaterialExpressionReroute_0}]
   properties {materialFunction=&BreakOutFloat2Components}
  }
  MaterialExpressionMax_11 {
   class=MaterialExpressionMax
   inputs [
    A {from=MaterialExpressionAdd_15}
    B {from=MaterialExpressionMaterialFunctionCall_2, output=G}
   ]
   properties {constB=0}
  }
  MaterialExpressionMax_12 {
   class=MaterialExpressionMax
   inputs [
    A {from=MaterialExpressionMaterialFunctionCall_7, output=R}
    B {from=MaterialExpressionMaterialFunctionCall_7, output=R}
   ]
   properties {constB=0}
  }
  MaterialExpressionMax_13 {
   class=MaterialExpressionMax
   inputs [
    A {from=MaterialExpressionMax_11}
    B {from=MaterialExpressionMax_12}
   ]
   properties {constB=0}
  }
  MaterialExpressionMin_0 {
   class=MaterialExpressionMin
   inputs [
    A {from=MaterialExpressionCeil_1, output=""}
    B {from=MaterialExpressionNamedRerouteUsage_5, output=""}
   ]
   properties {}
  }
  MaterialExpressionMultiply_13 {
   class=MaterialExpressionMultiply
   inputs [
    A {from=MaterialExpressionScalarParameter_27}
    B {from=""}
   ]
   properties {constB=-1}
  }
  MaterialExpressionMultiply_16 {
   class=MaterialExpressionMultiply
   inputs [
    A {from=MaterialExpressionScalarParameter_23}
    B {from=""}
   ]
   properties {constB=-1}
  }
  MaterialExpressionMultiply_20 {
   class=MaterialExpressionMultiply
   inputs [
    A {from=MaterialExpressionMin_0}
    B {from=MaterialExpressionFunctionInput_0}
   ]
  }
  MaterialExpressionMultiply_21 {
   class=MaterialExpressionMultiply
   inputs [
    A {from=MaterialExpressionMaterialFunctionCall_1, output=Result}
    B {from=MaterialExpressionSaturate_19}
   ]
  }
  MaterialExpressionMultiply_22 {
   class=MaterialExpressionMultiply
   inputs [
    A {from=MaterialExpressionScalarParameter_45}
    B {from=MaterialExpressionReroute_5}
   ]
  }
  MaterialExpressionMultiply_27 {
   class=MaterialExpressionMultiply
   inputs [
    A {from=MaterialExpressionNamedRerouteUsage_3}
    B {from=MaterialExpressionConstant2Vector_0}
   ]
  }
  MaterialExpressionMultiply_3 {
   class=MaterialExpressionMultiply
   inputs [
    A {from=MaterialExpressionFloor_0}
    B {from=""}
   ]
   properties {constB=-1}
  }
  MaterialExpressionMultiply_35 {
   class=MaterialExpressionMultiply
   inputs [
    A {from=MaterialExpressionSubtract_0}
    B {from=""}
   ]
   properties {constB=2}
  }
  MaterialExpressionMultiply_37 {
   class=MaterialExpressionMultiply
   inputs [
    A {from=MaterialExpressionScalarParameter_58}
    B {from=""}
   ]
   properties {constB=-1}
  }
  MaterialExpressionMultiply_40 {
   class=MaterialExpressionMultiply
   inputs [
    A {from=MaterialExpressionClamp_2}
    B {from=MaterialExpressionScalarParameter_21}
   ]
  }
  MaterialExpressionMultiply_41 {
   class=MaterialExpressionMultiply
   inputs [
    A {from=MaterialExpressionClamp_3}
    B {from=MaterialExpressionScalarParameter_18}
   ]
  }
  MaterialExpressionMultiply_5 {
   class=MaterialExpressionMultiply
   inputs [
    A {from=MaterialExpressionAdd_16}
    B {from=MaterialExpressionAdd_2}
   ]
  }
  MaterialExpressionMultiply_6 {
   class=MaterialExpressionMultiply
   inputs [
    A {from=MaterialExpressionOneMinus_2}
    B {from=MaterialExpressionNamedRerouteDeclaration_2}
   ]
   properties {desc=RightEye}
  }
  MaterialExpressionNamedRerouteDeclaration_0 {
   class=MaterialExpressionNamedRerouteDeclaration
   inputs [None {from=MaterialExpressionStaticSwitch_2}]
   properties {name=EyeUVs, nodeColor=<rgba 0.2188 0 1 1>}
  }
  MaterialExpressionNamedRerouteDeclaration_12 {
   class=MaterialExpressionNamedRerouteDeclaration
   inputs [None {from=MaterialExpressionOneMinus_1}]
   properties {name=Mask_EyesUDIM, nodeColor=<rgba 1 0.2578 0 1>}
  }
  MaterialExpressionNamedRerouteDeclaration_2 {
   class=MaterialExpressionNamedRerouteDeclaration
   inputs [None {from=MaterialExpressionMultiply_20}]
   properties {name=Mask_Eyes, nodeColor=<rgba 1 0 0.8906 1>}
  }
  MaterialExpressionNamedRerouteDeclaration_24 {
   class=MaterialExpressionNamedRerouteDeclaration
   inputs [None {from=MaterialExpressionLinearInterpolate_13}]
   properties {name=EyeCoords, nodeColor=<rgba 0 1 0.6016 1>}
  }
  MaterialExpressionNamedRerouteDeclaration_3 {
   class=MaterialExpressionNamedRerouteDeclaration
   inputs [None {from=MaterialExpressionMultiply_6}]
   properties {name=Mask_EyeRight, nodeColor=<rgba 1 0 0.0152 1>}
  }
  MaterialExpressionNamedRerouteDeclaration_32 {
   class=MaterialExpressionNamedRerouteDeclaration
   inputs [None {from=MaterialExpressionGetMaterialAttributes_0, output=MaterialAttributes}]
   properties {name=IN_Mat, nodeColor=<rgba 1 0 0.5156 1>}
  }
  MaterialExpressionNamedRerouteDeclaration_33 {
   class=MaterialExpressionNamedRerouteDeclaration
   inputs [None {from=MaterialExpressionGetMaterialAttributes_0, output=Normal}]
   properties {name=IN_Normal, nodeColor=<rgba 0 0.8359 1 1>}
  }
  MaterialExpressionNamedRerouteDeclaration_34 {
   class=MaterialExpressionNamedRerouteDeclaration
   inputs [None {from=MaterialExpressionMaterialFunctionCall_13, output=RGB}]
   properties {name=IN_SRM, nodeColor=<rgba 1 0.3516 0 1>}
  }
  MaterialExpressionNamedRerouteDeclaration_4 {
   class=MaterialExpressionNamedRerouteDeclaration
   inputs [None {from=MaterialExpressionFloor_1}]
   properties {name=Mask_EyeLeft, nodeColor=<rgba 0.0293 0.3125 0.0353 1>}
  }
  MaterialExpressionNamedRerouteUsage_0 {class=MaterialExpressionNamedRerouteUsage}
  MaterialExpressionNamedRerouteUsage_1 {class=MaterialExpressionNamedRerouteUsage}
  MaterialExpressionNamedRerouteUsage_2 {class=MaterialExpressionNamedRerouteUsage}
  MaterialExpressionNamedRerouteUsage_22 {class=MaterialExpressionNamedRerouteUsage}
  MaterialExpressionNamedRerouteUsage_23 {class=MaterialExpressionNamedRerouteUsage}
  MaterialExpressionNamedRerouteUsage_24 {class=MaterialExpressionNamedRerouteUsage}
  MaterialExpressionNamedRerouteUsage_3 {class=MaterialExpressionNamedRerouteUsage}
  MaterialExpressionNamedRerouteUsage_31 {class=MaterialExpressionNamedRerouteUsage}
  MaterialExpressionNamedRerouteUsage_32 {class=MaterialExpressionNamedRerouteUsage}
  MaterialExpressionNamedRerouteUsage_4 {class=MaterialExpressionNamedRerouteUsage}
  MaterialExpressionNamedRerouteUsage_40 {class=MaterialExpressionNamedRerouteUsage}
  MaterialExpressionNamedRerouteUsage_44 {class=MaterialExpressionNamedRerouteUsage}
  MaterialExpressionNamedRerouteUsage_45 {class=MaterialExpressionNamedRerouteUsage}
  MaterialExpressionNamedRerouteUsage_46 {class=MaterialExpressionNamedRerouteUsage}
  MaterialExpressionNamedRerouteUsage_47 {class=MaterialExpressionNamedRerouteUsage}
  MaterialExpressionNamedRerouteUsage_5 {class=MaterialExpressionNamedRerouteUsage}
  MaterialExpressionNamedRerouteUsage_52 {class=MaterialExpressionNamedRerouteUsage}
  MaterialExpressionNamedRerouteUsage_53 {class=MaterialExpressionNamedRerouteUsage}
  MaterialExpressionNamedRerouteUsage_55 {class=MaterialExpressionNamedRerouteUsage}
  MaterialExpressionNamedRerouteUsage_57 {class=MaterialExpressionNamedRerouteUsage}
  MaterialExpressionNamedRerouteUsage_58 {class=MaterialExpressionNamedRerouteUsage}
  MaterialExpressionNamedRerouteUsage_59 {class=MaterialExpressionNamedRerouteUsage}
  MaterialExpressionOneMinus_1 {class=MaterialExpressionOneMinus, inputs [None {from=MaterialExpressionMax_13}]}
  MaterialExpressionOneMinus_2 {class=MaterialExpressionOneMinus, inputs [None {from=MaterialExpressionFloor_1}]}
  MaterialExpressionPower_5 {
   class=MaterialExpressionPower
   inputs [
    Base {from=MaterialExpressionLinearInterpolate_19}
    Exp {from=""}
   ]
   properties {constExponent=0.5}
  }
  MaterialExpressionReroute_0 {class=MaterialExpressionReroute, inputs [None {from=MaterialExpressionMaterialFunctionCall_11, output=Result}]}
  MaterialExpressionReroute_3 {class=MaterialExpressionReroute, inputs [None {from=MaterialExpressionStaticSwitch_2}]}
  MaterialExpressionReroute_5 {class=MaterialExpressionReroute, inputs [None {from=MaterialExpressionNamedRerouteUsage_44}]}
  MaterialExpressionRotator_2 {
   class=MaterialExpressionRotator
   inputs [
    Coordinate {from=MaterialExpressionNamedRerouteUsage_2, output=""}
    Time {from=MaterialExpressionMultiply_37, output=""}
   ]
   properties {speed=0.5}
  }
  MaterialExpressionRotator_3 {
   class=MaterialExpressionRotator
   inputs [
    Coordinate {from=MaterialExpressionNamedRerouteUsage_2, output=""}
    Time {from=MaterialExpressionScalarParameter_59, output=""}
   ]
   properties {centerX=1, speed=0.5}
  }
  MaterialExpressionSaturate_19 {class=MaterialExpressionSaturate, inputs [None {from=MaterialExpressionInverseLinearInterpolate_3}], properties {desc=FOV compensation}}
  MaterialExpressionScalarParameter_0 {
   class=MaterialExpressionScalarParameter
   properties {defaultValue=0.5, group=01_EyeAnimation_Parameters, parameterName=PupilMap_Scale, sliderMax=2, sortPriority=5}
  }
  MaterialExpressionScalarParameter_1 {
   class=MaterialExpressionScalarParameter
   properties {group=01_EyeAnimation_Poses, parameterName=r_eye_vertical_offset, sliderMax=1, sortPriority=14}
  }
  MaterialExpressionScalarParameter_18 {
   class=MaterialExpressionScalarParameter
   properties {defaultValue=0.2, group=01_EyeAnimation_Parameters, parameterName=EyeVertical_Threshold, sliderMax=1}
  }
  MaterialExpressionScalarParameter_2 {
   class=MaterialExpressionScalarParameter
   properties {group=01_EyeAnimation_Poses, parameterName=l_eye_vertical_offset, sliderMax=1, sortPriority=14}
  }
  MaterialExpressionScalarParameter_20 {class=MaterialExpressionScalarParameter, properties {group=01_EyeAnimation_Poses, parameterName=look_up_pose, sliderMax=1, sortPriority=12}}
  MaterialExpressionScalarParameter_21 {
   class=MaterialExpressionScalarParameter
   properties {defaultValue=0.2, group=01_EyeAnimation_Parameters, parameterName=EyeHorizontal_Threshold, sliderMax=1}
  }
  MaterialExpressionScalarParameter_23 {
   class=MaterialExpressionScalarParameter
   properties {group=01_EyeAnimation_Poses, parameterName=look_right_pose, sliderMax=1, sortPriority=10}
  }
  MaterialExpressionScalarParameter_26 {
   class=MaterialExpressionScalarParameter
   properties {group=01_EyeAnimation_Poses, parameterName=look_left_pose, sliderMax=1, sortPriority=11}
  }
  MaterialExpressionScalarParameter_27 {
   class=MaterialExpressionScalarParameter
   properties {group=01_EyeAnimation_Poses, parameterName=look_down_pose, sliderMax=1, sortPriority=13}
  }
  MaterialExpressionScalarParameter_3 {
   class=MaterialExpressionScalarParameter
   properties {group=01_EyeAnimation_Poses, parameterName=r_eye_horizontal_offset, sliderMax=1, sortPriority=14}
  }
  MaterialExpressionScalarParameter_30 {
   class=MaterialExpressionScalarParameter
   properties {defaultValue=0.5, group=01_EyeAnimation_Parameters, parameterName=PupilMap_ScaleFar, sliderMax=2, sortPriority=5}
  }
  MaterialExpressionScalarParameter_31 {
   class=MaterialExpressionScalarParameter
   properties {defaultValue=500, group=01_EyeAnimation_Parameters, parameterName=PupilDistanceScaling_StartDistance, sliderMax=2, sortPriority=5}
  }
  MaterialExpressionScalarParameter_32 {
   class=MaterialExpressionScalarParameter
   properties {defaultValue=500, group=01_EyeAnimation_Parameters, parameterName=PupilDistanceScaling_FadeLength, sliderMax=2, sortPriority=5}
  }
  MaterialExpressionScalarParameter_34 {class=MaterialExpressionScalarParameter, properties {defaultValue=1, group=01_EyeAnimation_Parameters, parameterName=RotateRightEyeUVs}}
  MaterialExpressionScalarParameter_36 {
   class=MaterialExpressionScalarParameter
   properties {defaultValue=0.2, group=01_EyeAnimation_Parameters, parameterName=EyeVertical_Threshold, sliderMax=1}
  }
  MaterialExpressionScalarParameter_4 {
   class=MaterialExpressionScalarParameter
   properties {group=01_EyeAnimation_Poses, parameterName=l_eye_horizontal_offset, sliderMax=1, sortPriority=14}
  }
  MaterialExpressionScalarParameter_40 {
   class=MaterialExpressionScalarParameter
   properties {defaultValue=0.5, group=01_EyeAnimation_Parameters, parameterName=PupilMap_Scale_L, sliderMax=2, sortPriority=5}
  }
  MaterialExpressionScalarParameter_41 {
   class=MaterialExpressionScalarParameter
   properties {defaultValue=0.5, group=01_EyeAnimation_Parameters, parameterName=PupilMap_Scale_R, sliderMax=2, sortPriority=5}
  }
  MaterialExpressionScalarParameter_43 {
   class=MaterialExpressionScalarParameter
   properties {defaultValue=2, group=01_EyeAnimation_Parameters, parameterName=PupilMap_UV_Ratio_R, sliderMax=2, sortPriority=5}
  }
  MaterialExpressionScalarParameter_44 {
   class=MaterialExpressionScalarParameter
   properties {defaultValue=2, group=01_EyeAnimation_Parameters, parameterName=PupilMap_UV_Ratio_L, sliderMax=2, sortPriority=5}
  }
  MaterialExpressionScalarParameter_45 {
   class=MaterialExpressionScalarParameter
   properties {
    desc="If the base head AnimBP parent's \"Has Material-driven Pupils\" flag is true, this is set to either 0 or 1 depending on whether the pupil look direction has changed this frame or not.\r\nUse this to set the TSR invalidation amount (see the Temporal Responsiveness special material output node's tooltip for more details)"
    group=01_EyeAnimation_AnimData
    parameterName=HasLookDirectionChanged
   }
  }
  MaterialExpressionScalarParameter_46 {
   class=MaterialExpressionScalarParameter
   properties {group=01_EyeAnimation_Poses, parameterName=L_pupil_offsetX, sliderMax=2, sortPriority=20}
  }
  MaterialExpressionScalarParameter_47 {
   class=MaterialExpressionScalarParameter
   properties {group=01_EyeAnimation_Poses, parameterName=L_pupil_offsetY, sliderMax=2, sortPriority=20}
  }
  MaterialExpressionScalarParameter_48 {
   class=MaterialExpressionScalarParameter
   properties {group=01_EyeAnimation_Poses, parameterName=R_pupil_offsetX, sliderMax=2, sortPriority=20}
  }
  MaterialExpressionScalarParameter_49 {
   class=MaterialExpressionScalarParameter
   properties {group=01_EyeAnimation_Poses, parameterName=R_pupil_offsetY, sliderMax=2, sortPriority=20}
  }
  MaterialExpressionScalarParameter_5 {class=MaterialExpressionScalarParameter, properties {group=01_EyeAnimation_Poses, parameterName=pupil_squash_offset}}
  MaterialExpressionScalarParameter_51 {
   class=MaterialExpressionScalarParameter
   properties {defaultValue=0.35, group=01_EyeAnimation_Parameters, parameterName=EyeVertical_ThresholdMax, sliderMax=1}
  }
  MaterialExpressionScalarParameter_58 {class=MaterialExpressionScalarParameter, properties {group=01_EyeAnimation_Parameters, parameterName=PupilRotation_R}}
  MaterialExpressionScalarParameter_59 {class=MaterialExpressionScalarParameter, properties {group=01_EyeAnimation_Parameters, parameterName=PupilRotation_L}}
  MaterialExpressionScalarParameter_7 {
   class=MaterialExpressionScalarParameter
   properties {defaultValue=2, desc=horizontal UV ratio, group=01_EyeAnimation_Parameters, parameterName=PupilMap_UV_Ratio, sliderMax=2, sortPriority=5}
  }
  MaterialExpressionSetMaterialAttributes_0 {
   class=MaterialExpressionSetMaterialAttributes
   inputs [
    MaterialAttributes {from=MaterialExpressionNamedRerouteUsage_52, output=""}
    BaseColor {from=MaterialExpressionLinearInterpolate_29, output=""}
    Metallic {from=MaterialExpressionComponentMask_0, output=""}
    Specular {from=MaterialExpressionComponentMask_1, output=""}
    Roughness {from=MaterialExpressionComponentMask_4, output=""}
    Normal {from=MaterialExpressionStaticSwitchParameter_13, output=""}
   ]
   properties {attributeSetTypes [
     #5
     #4
     #2
     #3
     #1
    ]}
  }
  MaterialExpressionStaticBool_0 {class=MaterialExpressionStaticBool, properties {value=true}}
  MaterialExpressionStaticSwitchParameter_0 {
   class=MaterialExpressionStaticSwitchParameter
   inputs [
    True {from=MaterialExpressionLinearInterpolate_0, output=""}
    False {from=MaterialExpressionStaticSwitchParameter_10, output=""}
   ]
   properties {group=01_EyeAnimation_Parameters, parameterName=UsePupilDistanceScaling, sortPriority=4}
  }
  MaterialExpressionStaticSwitchParameter_10 {
   class=MaterialExpressionStaticSwitchParameter
   inputs [
    True {from=MaterialExpressionLinearInterpolate_22, output=""}
    False {from=MaterialExpressionScalarParameter_0, output=""}
   ]
   properties {group=01_EyeAnimation_Parameters, parameterName=UsePupilSeparateScalingRatio, sortPriority=4}
  }
  MaterialExpressionStaticSwitchParameter_11 {
   class=MaterialExpressionStaticSwitchParameter
   inputs [
    True {from=MaterialExpressionLinearInterpolate_23, output=""}
    False {from=MaterialExpressionScalarParameter_7, output=""}
   ]
   properties {group=01_EyeAnimation_Parameters, parameterName=UsePupilSeparateScalingRatio, sortPriority=4}
  }
  MaterialExpressionStaticSwitchParameter_12 {
   class=MaterialExpressionStaticSwitchParameter
   inputs [
    True {from=MaterialExpressionLinearInterpolate_24, output=""}
    False {from=MaterialExpressionConstant_3, output=""}
   ]
   properties {group=01_EyeAnimation_Parameters, parameterName=UseEyeAnimation, sortPriority=-1}
  }
  MaterialExpressionStaticSwitchParameter_13 {
   class=MaterialExpressionStaticSwitchParameter
   inputs [
    True {from=MaterialExpressionMaterialFunctionCall_8, output=Result}
    False {from=MaterialExpressionNamedRerouteUsage_57, output=""}
   ]
   properties {group=01_EyeAnimation_Parameters, parameterName=UsePupilNormal, sortPriority=0}
  }
  MaterialExpressionStaticSwitchParameter_14 {
   class=MaterialExpressionStaticSwitchParameter
   inputs [
    True {from=MaterialExpressionBlendMaterialAttributes_0, output=""}
    False {from=MaterialExpressionNamedRerouteUsage_58, output=""}
   ]
   properties {group=01_EyeAnimation_Parameters, parameterName=UseEyeAnimation, sortPriority=-1}
  }
  MaterialExpressionStaticSwitchParameter_15 {
   class=MaterialExpressionStaticSwitchParameter
   inputs [
    True {from=MaterialExpressionTextureSampleParameter2D_2, output=RGB}
    False {from=MaterialExpressionNamedRerouteUsage_59, output=""}
   ]
   properties {group=01_EyeAnimation_Parameters, parameterName=UsePupilSRM, sortPriority=0}
  }
  MaterialExpressionStaticSwitchParameter_8 {
   class=MaterialExpressionStaticSwitchParameter
   inputs [
    True {from=MaterialExpressionLinearInterpolate_25, output=""}
    False {from=MaterialExpressionScalarParameter_36, output=""}
   ]
   properties {group=01_EyeAnimation_Parameters, parameterName=UseEyeThresholdLidInfluenced, sortPriority=40}
  }
  MaterialExpressionStaticSwitch_2 {
   class=MaterialExpressionStaticSwitch
   inputs [
    True {from=MaterialExpressionTextureCoordinate_5, output=""}
    False {from=MaterialExpressionTextureCoordinate_4, output=""}
    Value {from=MaterialExpressionFunctionInput_1, output=""}
   ]
   properties {}
  }
  MaterialExpressionSubtract_0 {
   class=MaterialExpressionSubtract
   inputs [
    A {from=MaterialExpressionMaterialFunctionCall_10, output=R}
    B {from=""}
   ]
   properties {constB=0.5}
  }
  MaterialExpressionTextureCoordinate_4 {class=MaterialExpressionTextureCoordinate}
  MaterialExpressionTextureCoordinate_5 {class=MaterialExpressionTextureCoordinate, properties {coordinateIndex=1}}
  MaterialExpressionTextureSampleParameter2D_1 {
   class=MaterialExpressionTextureSampleParameter2D
   inputs [
    UVs {from=MaterialExpressionNamedRerouteUsage_45, output=""}
    Apply View MipBias {from="", output=""}
   ]
   properties {
    group=01_EyeAnimation_Parameters
    parameterName=Pupil_Diffuse
    samplerSource=SSM_Clamp_WorldGroupSettings
    sortPriority=1
    texture=/Game/Global/Textures/Default/Blanks/T_MidgreysRGB
   }
  }
  MaterialExpressionTextureSampleParameter2D_2 {
   class=MaterialExpressionTextureSampleParameter2D
   inputs [
    UVs {from=MaterialExpressionNamedRerouteUsage_46, output=""}
    Apply View MipBias {from="", output=""}
   ]
   properties {
    group=01_EyeAnimation_Parameters
    parameterName=Pupil_SRM
    samplerSource=SSM_Clamp_WorldGroupSettings
    samplerType=SAMPLERTYPE_Masks
    sortPriority=1
    texture=&T_Fortnite_Default_S
   }
  }
  MaterialExpressionTextureSampleParameter2D_3 {
   class=MaterialExpressionTextureSampleParameter2D
   inputs [
    UVs {from=MaterialExpressionNamedRerouteUsage_47, output=""}
    Apply View MipBias {from="", output=""}
   ]
   properties {
    group=01_EyeAnimation_Parameters
    parameterName=Pupil_N
    samplerSource=SSM_Clamp_WorldGroupSettings
    samplerType=SAMPLERTYPE_Normal
    sortPriority=1
    texture=&FlatNormalBlue
   }
  }
  MaterialExpressionVectorParameter_5 {
   class=MaterialExpressionVectorParameter
   properties {group=01_EyeAnimation_Parameters, parameterName=PupilOffset_L, defaultValue=<rgba 0 0 0 1>}
  }
  MaterialExpressionVectorParameter_6 {
   class=MaterialExpressionVectorParameter
   properties {group=01_EyeAnimation_Parameters, parameterName=PupilOffset_R, defaultValue=<rgba 0 0 0 1>}
  }
 }}
