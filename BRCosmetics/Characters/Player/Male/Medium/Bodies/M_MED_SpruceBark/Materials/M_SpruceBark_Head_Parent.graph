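# Parent material for the SpruceBark character's head, layering fur and animated eyes onto the 2023 character parent pipeline.
# PreFX feeds named reroutes for material and masks, used around a fur chain of custom color, parameter adjustments, painting, shells, WPO shrink and subsurface.
# A SpruceBark eye animation function using UV2 feeds PostFX and also drives the temporal responsiveness output.
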Material /BRCosmetics/Characters/Player/Male/Medium/Bodies/M_MED_SpruceBark/Materials/M_SpruceBark_Head_Parent
graph {
 properties {
  bUseMaterialAttributes=true
  bUsedWithClothing=true
  bUsedWithInstancedStaticMeshes=true
  bUsedWithMeshDeformer=true
  bUsedWithMorphTargets=true
  bUsedWithNanite=true
  bUsedWithSkeletalMesh=true
  pixelDepthOffsetMode=PDOM_Legacy
  previewMesh=/BRCosmetics/Characters/Player/Female/Medium/Bodies/F_MED_PhoenixKick_Gain/Meshes/F_MED_PhoenixKick_Gain
  shadingModel=MSM_FromMaterialExpression
  usedShadingModels=Default Lit | Subsurface | Clear Coat | Subsurface Profile | Cloth
 }
 nodes [
  Constant()
  n1 = MaterialFunctionCall!(Material=NamedRerouteUsage(), Use UV2=StaticBool() {value=true}, Mask) {materialFunction=&/MF_PBR_EyeAnimation_SpruceBark}
  n2 = MaterialFunctionCall() {materialFunction=&MF_CharacterParent_2023_PreFX}
  MaterialFunctionCall(Material=@n1.Result, M=@n2.M, Pre Skinned Local Position=@n2.M, Distance Blend=@n2.M, Fresnel=@n2.M) {materialFunction=&MF_CharacterParent_2023_PostFX}
  NamedRerouteDeclaration(None=@n2.Tech Art Mask) {name=TechArtMask, nodeColor=<rgba 0.2188 0 1 1>}
  NamedRerouteDeclaration(None=@n2.Result) {name=Material, nodeColor=<rgba 0 1 0.1094 1>}
  NamedRerouteDeclaration(None=@n2.M) {name=Mask, nodeColor=<rgba 1 0 0.2109 1>}
  n3 = MaterialFunctionCall!(Material=NamedRerouteUsage(), TechArtMask=NamedRerouteUsage()) {materialFunction=/Game/Global/Materials/Character_Parent_2023/Fur_Parent_2023/MF_Fur_CustomColor}
  n4 = NamedRerouteUsage()
  n5 = MaterialFunctionCall!(
   In=MaterialFunctionCall!(
    In=MaterialFunctionCall!(MA=@n3.Result, TechArtMask=@n4) {materialFunction=&/MF_Fur_ParamAdjustments_SpruceBark}.Result
    TechArtMask=NamedRerouteUsage()
    Mask Input=NamedRerouteUsage()
   ) {materialFunction=/Game/Global/Materials/Character_Parent_2023/Fur_Parent_2023/MF_Fur_Painted}.Result
   TechArtMask=NamedRerouteUsage()
  ) {materialFunction=/Game/Global/Materials/Character_Parent_2023/Fur_Parent_2023/MF_Fur_Shells}
  n6 = NamedRerouteUsage()
  NamedRerouteDeclaration(None=MaterialFunctionCall!(
    MA=MaterialFunctionCall!(Material=@n5.Result, Mask=@n6, Fur Length=@n5.Result) {materialFunction=/Game/Global/Materials/Character_Parent_2023/Fur_Parent_2023/MF_FurWPOShrink}.Result
    TechArtMask=NamedRerouteUsage()
   ) {materialFunction=&/MF_Fur_Subsurface_SpruceBark}.Result) {name=OUT_Material, nodeColor=<rgba 0 0.5312 1 1>}
  NamedRerouteDeclaration(None=@n2.Fresnel) {name=Fresnel, nodeColor=<rgba 1 0.6563 0 1>}
  TemporalResponsivenessOutput!(None=@n1.TemporalResponsiveness)
 ]
}
